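# Animation Blueprint for the HiHat pickaxe that adds springy, tightly limited secondary motion to its dyn_chain bones.
# The input pose passes through a Control Rig, fed the melee-anim flag, into a cached pose.
# Nine serial Anim Dynamics nodes (center, left, right) then simulate bones relative to the two-handed handle.
# On initialization, a front-end check selects the shared gravity and sim-space settings, with weaker gravity and external spin in the front end.
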AnimBlueprint /BRCosmetics/Weapons/FORT_Melee/Pickaxe_HiHat/Meshes/HiHat_Axe_AnimBP
target_skeleton=&/HiHat_Axe_Skeleton
graphs [
 AnimGraph {nodes [
   n1 = Get FEGravity() -> (FEGravity:Vector)
   n2 = Get Sim Setting() -> (Sim Setting:Anim Phys Sim Space Settings Structure)
   n3 = Anim Dynamics - Bone: dyn_chain01_cr!(
    ComponentPose=Local To Component(LocalPose=Use cached pose 'StartPose'())
    Alpha=1
    GravityOverride:Vector=@n1
    SimSpaceSettings:Anim Phys Sim Space Settings Structure=@n2
    ExternalForce="0,0,0"
   ) -> (Pose) {
    angularDampingOverride=1
    angularSpringConstant=3500
    bAngularSpring=true
    bOverrideAngularDamping=true
    bUseGravityOverride=true
    gravityScale=0.35
    linearSpringConstant=300
    simulationSpace=BoneRelative
    boundBone {boneName=dyn_chain01_cr}
    componentAppliedLinearAccClamp=<4000 4000 4000>
    componentLinearAccScale=<0.35 0.35 0.35>
    componentLinearVelScale=<0.35 0.35 0.35>
    gravityOverride {z=-700}
    physicsBodyDefinitions [{
      boundBone {boneName=dyn_chain01_cr}
      boxExtents=<20 20 20>
      constraintSetup {angularLimitsMax=<3 4 8>, angularLimitsMin=<-3 -4 -8>, angularTarget=<1 0 0>, linearAxesMax=<0 0.3 0>, linearAxesMin=<0 -0.3 0>}
      localJointOffset=<4 0 0>
     }]
    relativeSpaceBone {boneName=Melee_TwoHanded_Handle}
    simSpaceSettings {maxAngularAcceleration=10, maxAngularVelocity=200, simSpaceAngularAlpha=1}
   }
   n4 = Anim Dynamics - Bone: dyn_chain01_lf!(ComponentPose=@n3, Alpha=1, GravityOverride:Vector=@n1, SimSpaceSettings:Anim Phys Sim Space Settings Structure=@n2, ExternalForce="0,0,0") -> (Pose) {
    angularDampingOverride=1
    angularSpringConstant=3500
    bAngularSpring=true
    bOverrideAngularDamping=true
    bUseGravityOverride=true
    gravityScale=0.35
    linearSpringConstant=300
    simulationSpace=BoneRelative
    boundBone {boneName=dyn_chain01_lf}
    componentAppliedLinearAccClamp=<4000 4000 4000>
    componentLinearAccScale=<0.35 0.35 0.35>
    componentLinearVelScale=<0.35 0.35 0.35>
    gravityOverride {z=-700}
    physicsBodyDefinitions [{
      boundBone {boneName=dyn_chain01_lf}
      boxExtents=<20 20 20>
      constraintSetup {angularLimitsMax=<3 4 8>, angularLimitsMin=<-3 -4 -8>, angularTarget=<1 0 0>, linearAxesMax=<0 0.3 0>, linearAxesMin=<0 -0.3 0>}
      localJointOffset=<4 0 0>
     }]
    relativeSpaceBone {boneName=Melee_TwoHanded_Handle}
    simSpaceSettings {maxAngularAcceleration=10, maxAngularVelocity=200, simSpaceAngularAlpha=1}
   }
   n5 = Anim Dynamics - Bone: dyn_chain02_lf!(ComponentPose=@n4, Alpha=1, GravityOverride:Vector=@n1, SimSpaceSettings:Anim Phys Sim Space Settings Structure=@n2, ExternalForce="0,0,0") -> (Pose) {
    angularDampingOverride=1
    angularSpringConstant=3500
    bAngularSpring=true
    bOverrideAngularDamping=true
    bUseGravityOverride=true
    gravityScale=0.35
    linearSpringConstant=300
    simulationSpace=BoneRelative
    boundBone {boneName=dyn_chain02_lf}
    componentAppliedLinearAccClamp=<4000 4000 4000>
    componentLinearAccScale=<0.35 0.35 0.35>
    componentLinearVelScale=<0.35 0.35 0.35>
    gravityOverride {z=-700}
    physicsBodyDefinitions [{
      boundBone {boneName=dyn_chain02_lf}
      boxExtents=<23 23 23>
      constraintSetup {angularLimitsMax=<3 4 7>, angularLimitsMin=<-3 -4 -7>, angularTarget=<1 0 0>, linearAxesMax=<0 0.3 0>, linearAxesMin=<0 -0.3 0>}
      localJointOffset=<4 0 0>
     }]
    relativeSpaceBone {boneName=Melee_TwoHanded_Handle}
    simSpaceSettings {maxAngularAcceleration=10, maxAngularVelocity=200, simSpaceAngularAlpha=1}
   }
   n6 = Anim Dynamics - Bone: dyn_chain03_lf!(ComponentPose=@n5, Alpha=1, GravityOverride:Vector=@n1, SimSpaceSettings:Anim Phys Sim Space Settings Structure=@n2, ExternalForce="0,0,0") -> (Pose) {
    angularDampingOverride=1
    angularSpringConstant=3500
    bAngularSpring=true
    bOverrideAngularDamping=true
    bUseGravityOverride=true
    gravityScale=0.35
    linearSpringConstant=300
    simulationSpace=BoneRelative
    boundBone {boneName=dyn_chain03_lf}
    componentAppliedLinearAccClamp=<4000 4000 4000>
    componentLinearAccScale=<0.35 0.35 0.35>
    componentLinearVelScale=<0.35 0.35 0.35>
    gravityOverride {z=-700}
    physicsBodyDefinitions [{
      boundBone {boneName=dyn_chain03_lf}
      boxExtents=<20 20 20>
      constraintSetup {angularLimitsMax=<3 4 8>, angularLimitsMin=<-3 -4 -8>, angularTarget=<1 0 0>, linearAxesMax=<0 0.3 0>, linearAxesMin=<0 -0.3 0>}
      localJointOffset=<4 0 0>
     }]
    relativeSpaceBone {boneName=Melee_TwoHanded_Handle}
    simSpaceSettings {maxAngularAcceleration=10, maxAngularVelocity=200, simSpaceAngularAlpha=1}
   }
   n7 = Anim Dynamics - Bone: dyn_chain04_lf!(ComponentPose=@n6, Alpha=1, GravityOverride:Vector=@n1, SimSpaceSettings:Anim Phys Sim Space Settings Structure=@n2, ExternalForce="0,0,0") -> (Pose) {
    angularDampingOverride=1
    angularSpringConstant=3500
    bAngularSpring=true
    bOverrideAngularDamping=true
    bUseGravityOverride=true
    gravityScale=0.35
    linearSpringConstant=300
    simulationSpace=BoneRelative
    boundBone {boneName=dyn_chain04_lf}
    componentAppliedLinearAccClamp=<4000 4000 4000>
    componentLinearAccScale=<0.35 0.35 0.35>
    componentLinearVelScale=<0.35 0.35 0.35>
    gravityOverride {z=-700}
    physicsBodyDefinitions [{
      boundBone {boneName=dyn_chain04_lf}
      boxExtents=<17 17 17>
      constraintSetup {angularLimitsMax=<3 4 9>, angularLimitsMin=<-3 -4 -9>, angularTarget=<1 0 0>, linearAxesMax=<0 0.3 0>, linearAxesMin=<0 -0.3 0>}
      localJointOffset=<4 0 0>
     }]
    relativeSpaceBone {boneName=Melee_TwoHanded_Handle}
    simSpaceSettings {maxAngularAcceleration=10, maxAngularVelocity=200, simSpaceAngularAlpha=1}
   }
   n8 = Anim Dynamics - Bone: dyn_chain01_rt!(ComponentPose=@n7, Alpha=1, GravityOverride:Vector=@n1, SimSpaceSettings:Anim Phys Sim Space Settings Structure=@n2, ExternalForce="0,0,0") -> (Pose) {
    angularDampingOverride=1
    angularSpringConstant=3500
    bAngularSpring=true
    bOverrideAngularDamping=true
    bUseGravityOverride=true
    gravityScale=0.35
    linearSpringConstant=300
    simulationSpace=BoneRelative
    boundBone {boneName=dyn_chain01_rt}
    componentAppliedLinearAccClamp=<4000 4000 4000>
    componentLinearAccScale=<0.35 0.35 0.35>
    componentLinearVelScale=<0.35 0.35 0.35>
    gravityOverride {z=-700}
    physicsBodyDefinitions [{
      boundBone {boneName=dyn_chain01_rt}
      boxExtents=<16 16 16>
      constraintSetup {angularLimitsMax=<3 4 9>, angularLimitsMin=<-3 -4 -9>, angularTarget=<1 0 0>, linearAxesMax=<0 0.3 0>, linearAxesMin=<0 -0.3 0>}
      localJointOffset=<4 0 0>
     }]
    relativeSpaceBone {boneName=Melee_TwoHanded_Handle}
    simSpaceSettings {maxAngularAcceleration=10, maxAngularVelocity=200, simSpaceAngularAlpha=1}
   }
   n9 = Anim Dynamics - Bone: dyn_chain02_rt!(ComponentPose=@n8, Alpha=1, GravityOverride:Vector=@n1, SimSpaceSettings:Anim Phys Sim Space Settings Structure=@n2, ExternalForce="0,0,0") -> (Pose) {
    angularDampingOverride=1
    angularSpringConstant=3500
    bAngularSpring=true
    bOverrideAngularDamping=true
    bUseGravityOverride=true
    gravityScale=0.35
    linearSpringConstant=300
    simulationSpace=BoneRelative
    boundBone {boneName=dyn_chain02_rt}
    componentAppliedLinearAccClamp=<4000 4000 4000>
    componentLinearAccScale=<0.35 0.35 0.35>
    componentLinearVelScale=<0.35 0.35 0.35>
    gravityOverride {z=-700}
    physicsBodyDefinitions [{
      boundBone {boneName=dyn_chain02_rt}
      boxExtents=<18 18 18>
      constraintSetup {angularLimitsMax=<3 4 8>, angularLimitsMin=<-3 -4 -8>, angularTarget=<1 0 0>, linearAxesMax=<0 0.3 0>, linearAxesMin=<0 -0.3 0>}
      localJointOffset=<4 0 0>
     }]
    relativeSpaceBone {boneName=Melee_TwoHanded_Handle}
    simSpaceSettings {maxAngularAcceleration=10, maxAngularVelocity=200, simSpaceAngularAlpha=1}
   }
   n10 = Anim Dynamics - Bone: dyn_chain03_rt!(ComponentPose=@n9, Alpha=1, GravityOverride:Vector=@n1, SimSpaceSettings:Anim Phys Sim Space Settings Structure=@n2, ExternalForce="0,0,0") -> (Pose) {
    angularDampingOverride=1
    angularSpringConstant=3500
    bAngularSpring=true
    bOverrideAngularDamping=true
    bUseGravityOverride=true
    gravityScale=0.35
    linearSpringConstant=300
    simulationSpace=BoneRelative
    boundBone {boneName=dyn_chain03_rt}
    componentAppliedLinearAccClamp=<4000 4000 4000>
    componentLinearAccScale=<0.35 0.35 0.35>
    componentLinearVelScale=<0.35 0.35 0.35>
    gravityOverride {z=-700}
    physicsBodyDefinitions [{
      boundBone {boneName=dyn_chain03_rt}
      boxExtents=<23 23 23>
      constraintSetup {angularLimitsMax=<3 4 7>, angularLimitsMin=<-3 -4 -7>, angularTarget=<1 0 0>, linearAxesMax=<0 0.3 0>, linearAxesMin=<0 -0.3 0>}
      localJointOffset=<4 0 0>
     }]
    relativeSpaceBone {boneName=Melee_TwoHanded_Handle}
    simSpaceSettings {maxAngularAcceleration=10, maxAngularVelocity=200, simSpaceAngularAlpha=1}
   }
   n11 = Anim Dynamics - Bone: dyn_chain04_rt!(ComponentPose=@n10, Alpha=1, GravityOverride:Vector=@n1, SimSpaceSettings:Anim Phys Sim Space Settings Structure=@n2, ExternalForce="0,0,0") -> (Pose) {
    angularDampingOverride=1
    angularSpringConstant=3500
    bAngularSpring=true
    bOverrideAngularDamping=true
    bUseGravityOverride=true
    gravityScale=0.35
    linearSpringConstant=300
    simulationSpace=BoneRelative
    boundBone {boneName=dyn_chain04_rt}
    componentAppliedLinearAccClamp=<4000 4000 4000>
    componentLinearAccScale=<0.35 0.35 0.35>
    componentLinearVelScale=<0.35 0.35 0.35>
    gravityOverride {z=-700}
    physicsBodyDefinitions [{
      boundBone {boneName=dyn_chain04_rt}
      boxExtents=<20 20 20>
      constraintSetup {angularLimitsMax=<3 4 8>, angularLimitsMin=<-3 -4 -8>, angularTarget=<1 0 0>, linearAxesMax=<0 0.3 0>, linearAxesMin=<0 -0.3 0>}
      localJointOffset=<4 0 0>
     }]
    relativeSpaceBone {boneName=Melee_TwoHanded_Handle}
    simSpaceSettings {maxAngularAcceleration=10, maxAngularVelocity=200, simSpaceAngularAlpha=1}
   }
   Output Pose(Result=Component To Local(ComponentPose=@n11)) {layerGroup=None}
   AnimGraphNode_SubInput_0 = Input Pose()
   Save cached pose 'StartPose'(Pose=Control Rig!(Source=@AnimGraphNode_SubInput_0, Alpha, Is Melee:bool=Get bIsPlayingMeleeAnim()) -> (Pose) {controlRigAssetReference {controlRigAsset=&/HiHat_Axe_CtrlRig}}) {cacheName=StartPose}
  ]}
 EventGraph {nodes [
   Event BlueprintInitializeAnimation()
   > n1 = Branch(Condition=Get bIsInFrontEnd())
   > Set FEGravity(FEGravity:Vector="0,0,-120") -> (Output_Get:Vector)
   > Set Sim Setting(Sim Setting:Anim Phys Sim Space Settings Structure=Get Sim Setting FE() -> (Sim Setting FE:Anim Phys Sim Space Settings Structure)) -> (Output_Get:Anim Phys Sim Space Settings Structure)
   Set FEGravity(execute=@n1.else, FEGravity="0,0,-700")
   > Set Sim Setting(Sim Setting=Get Sim Setting BR() -> (Sim Setting BR:Anim Phys Sim Space Settings Structure))
   TryGetPawnOwner()
   Event BlueprintUpdateAnimation()
  ]}
]
variables [
 FEGravity {type=Vector, default=<0 0 -700>}
 Sim Setting {
  type=AnimPhysSimSpaceSettings
  default {maxAngularAcceleration=10, maxAngularVelocity=200, simSpaceAngularAlpha=1, externalAngularVelocity=<0 0 0>}
 }
 Sim Setting BR {
  type=AnimPhysSimSpaceSettings
  default {maxAngularAcceleration=10, maxAngularVelocity=200, simSpaceAngularAlpha=1, externalAngularVelocity=<0 0 0>}
 }
 Sim Setting FE {
  type=AnimPhysSimSpaceSettings
  default {maxAngularAcceleration=10, maxAngularVelocity=200, simSpaceAngularAlpha=1, externalAngularVelocity=<0 0 3>}
 }
]
